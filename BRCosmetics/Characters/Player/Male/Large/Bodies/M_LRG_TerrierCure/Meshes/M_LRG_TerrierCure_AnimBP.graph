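# Post-process style AnimBP for the M_LRG_TerrierCure body cosmetic, adding secondary motion to its left and right dyn_pjs_bot bones.
# Parent-mesh pose copy feeds a Control Rig, then twelve chained spring-driven, limit-constrained AnimDynamics nodes.
# DefaultSlot layers over calf_twist_02_r, weighted by a front-end idle curve sampled on customization load.
# Skydiving and front-end events set spring, alpha and simulation-space variables, which the AnimGraph never reads.
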AnimBlueprint /BRCosmetics/Characters/Player/Male/Large/Bodies/M_LRG_TerrierCure/Meshes/M_LRG_TerrierCure_AnimBP
target_skeleton=&/M_LRG_TerrierCure_Skeleton
graphs [
 AnimGraph {nodes [
   n1 = Control Rig(Source=Copy Pose From Mesh() {bUseAttachedParent=true}) {
    alphaBoolBlend {blendOption=Linear}
    controlRigAssetReference {blueprintRigClass=&/M_LRG_TerrierCure__CtrlRig.M_LRG_TerrierCure__CtrlRig_C}
   }
   n2 = Anim Dynamics - Bone: dyn_pjs_bot_l_A!(
    SimulationSpace:AnimPhysSimSpaceType Enum
    ComponentPose=Local To Component(LocalPose=@n1)
    Alpha=1
    LinearSpringConstant:float=30
    ExternalForce="0,0,0"
   ) -> (Pose) {
    angularDampingOverride=0.9999
    angularSpringConstant=2000
    bAngularSpring=true
    bGravityOverrideInSimSpace=true
    bLinearSpring=true
    bOverrideAngularDamping=true
    bOverrideLinearDamping=true
    bUseGravityOverride=true
    linearDampingOverride=0.9999
    linearSpringConstant=30
    simulationSpace=RootRelative
    alphaBoolBlend {blendOption=Linear}
    boundBone {boneName=dyn_pjs_bot_l_A}
    componentAppliedLinearAccClamp=<1000 1000 1000>
    componentLinearAccScale=<1 1 1>
    componentLinearVelScale=<1 1 1>
    gravityOverride {z=-980}
    physicsBodyDefinitions [{
      boundBone {boneName=dyn_pjs_bot_l_A}
      boxExtents=<7 7 7>
      constraintSetup {angularLimitsMax=<5 5 5>, angularLimitsMin=<-5 -10 -5>, angularTarget=<0 0 0>, linearAxesMax=<2 2 4>, linearAxesMin=<-2 -2 0>}
      localJointOffset=<0.1 0 0>
     }]
   }
   n3 = Anim Dynamics - Bone: dyn_pjs_bot_l_B!(SimulationSpace:AnimPhysSimSpaceType Enum, ComponentPose=@n2, Alpha=1, LinearSpringConstant:float=30, ExternalForce="0,0,0") -> (Pose) {
    angularDampingOverride=0.9999
    angularSpringConstant=2000
    bAngularSpring=true
    bGravityOverrideInSimSpace=true
    bLinearSpring=true
    bOverrideAngularDamping=true
    bOverrideLinearDamping=true
    bUseGravityOverride=true
    linearDampingOverride=0.9999
    linearSpringConstant=30
    simulationSpace=RootRelative
    alphaBoolBlend {blendOption=Linear}
    boundBone {boneName=dyn_pjs_bot_l_B}
    componentAppliedLinearAccClamp=<1000 1000 1000>
    componentLinearAccScale=<1 1 1>
    componentLinearVelScale=<1 1 1>
    gravityOverride {z=-980}
    physicsBodyDefinitions [{
      boundBone {boneName=dyn_pjs_bot_l_B}
      boxExtents=<7 7 7>
      constraintSetup {angularLimitsMax=<5 5 5>, angularLimitsMin=<-5 -10 -5>, angularTarget=<0 0 0>, linearAxesMax=<2 2 4>, linearAxesMin=<-2 -2 0>}
      localJointOffset=<0.1 0 0>
     }]
   }
   n4 = Anim Dynamics - Bone: dyn_pjs_bot_l_C!(SimulationSpace:AnimPhysSimSpaceType Enum, ComponentPose=@n3, Alpha=1, LinearSpringConstant:float=30, ExternalForce="0,0,0") -> (Pose) {
    angularDampingOverride=0.9999
    angularSpringConstant=2000
    bAngularSpring=true
    bGravityOverrideInSimSpace=true
    bLinearSpring=true
    bOverrideAngularDamping=true
    bOverrideLinearDamping=true
    bUseGravityOverride=true
    linearDampingOverride=0.9999
    linearSpringConstant=30
    simulationSpace=RootRelative
    alphaBoolBlend {blendOption=Linear}
    boundBone {boneName=dyn_pjs_bot_l_C}
    componentAppliedLinearAccClamp=<1000 1000 1000>
    componentLinearAccScale=<1 1 1>
    componentLinearVelScale=<1 1 1>
    gravityOverride {z=-980}
    physicsBodyDefinitions [{
      boundBone {boneName=dyn_pjs_bot_l_C}
      boxExtents=<7 7 7>
      constraintSetup {angularLimitsMax=<5 5 5>, angularLimitsMin=<-5 -10 -5>, angularTarget=<0 0 0>, linearAxesMax=<2 2 4>, linearAxesMin=<-2 -2 0>}
      localJointOffset=<0.1 0 0>
     }]
   }
   n5 = Anim Dynamics - Bone: dyn_pjs_bot_l_D!(SimulationSpace:AnimPhysSimSpaceType Enum, ComponentPose=@n4, Alpha=1, LinearSpringConstant:float=30, ExternalForce="0,0,0") -> (Pose) {
    angularDampingOverride=0.9999
    angularSpringConstant=2000
    bAngularSpring=true
    bGravityOverrideInSimSpace=true
    bLinearSpring=true
    bOverrideAngularDamping=true
    bOverrideLinearDamping=true
    bUseGravityOverride=true
    linearDampingOverride=0.9999
    linearSpringConstant=30
    simulationSpace=RootRelative
    alphaBoolBlend {blendOption=Linear}
    boundBone {boneName=dyn_pjs_bot_l_D}
    componentAppliedLinearAccClamp=<1000 1000 1000>
    componentLinearAccScale=<1 1 1>
    componentLinearVelScale=<1 1 1>
    gravityOverride {z=-980}
    physicsBodyDefinitions [{
      boundBone {boneName=dyn_pjs_bot_l_D}
      boxExtents=<7 7 7>
      constraintSetup {angularLimitsMax=<5 5 5>, angularLimitsMin=<-5 -10 -5>, angularTarget=<0 0 0>, linearAxesMax=<2 2 4>, linearAxesMin=<-2 -2 0>}
      localJointOffset=<0.1 0 0>
     }]
   }
   n6 = Anim Dynamics - Bone: dyn_pjs_bot_l_E!(SimulationSpace:AnimPhysSimSpaceType Enum, ComponentPose=@n5, Alpha=1, LinearSpringConstant:float=30, ExternalForce="0,0,0") -> (Pose) {
    angularDampingOverride=0.9999
    angularSpringConstant=2000
    bAngularSpring=true
    bGravityOverrideInSimSpace=true
    bLinearSpring=true
    bOverrideAngularDamping=true
    bOverrideLinearDamping=true
    bUseGravityOverride=true
    linearDampingOverride=0.9999
    linearSpringConstant=30
    simulationSpace=RootRelative
    alphaBoolBlend {blendOption=Linear}
    boundBone {boneName=dyn_pjs_bot_l_E}
    componentAppliedLinearAccClamp=<1000 1000 1000>
    componentLinearAccScale=<1 1 1>
    componentLinearVelScale=<1 1 1>
    gravityOverride {z=-980}
    physicsBodyDefinitions [{
      boundBone {boneName=dyn_pjs_bot_l_E}
      boxExtents=<7 7 7>
      constraintSetup {angularLimitsMax=<5 5 5>, angularLimitsMin=<-5 -10 -5>, angularTarget=<0 0 0>, linearAxesMax=<2 2 5>, linearAxesMin=<-2 -2 0>}
      localJointOffset=<0.1 0 0>
     }]
   }
   n7 = Anim Dynamics - Bone: dyn_pjs_bot_l_F!(SimulationSpace:AnimPhysSimSpaceType Enum, ComponentPose=@n6, Alpha=1, LinearSpringConstant:float=30, ExternalForce="0,0,0") -> (Pose) {
    angularDampingOverride=0.9999
    angularSpringConstant=2000
    bAngularSpring=true
    bGravityOverrideInSimSpace=true
    bLinearSpring=true
    bOverrideAngularDamping=true
    bOverrideLinearDamping=true
    bUseGravityOverride=true
    linearDampingOverride=0.9999
    linearSpringConstant=30
    simulationSpace=RootRelative
    alphaBoolBlend {blendOption=Linear}
    boundBone {boneName=dyn_pjs_bot_l_F}
    componentAppliedLinearAccClamp=<1000 1000 1000>
    componentLinearAccScale=<1 1 1>
    componentLinearVelScale=<1 1 1>
    gravityOverride {z=-980}
    physicsBodyDefinitions [{
      boundBone {boneName=dyn_pjs_bot_l_F}
      boxExtents=<7 7 7>
      constraintSetup {angularLimitsMax=<5 5 5>, angularLimitsMin=<-5 -10 -5>, angularTarget=<0 0 0>, linearAxesMax=<2 2 4>, linearAxesMin=<-2 -2 0>}
      localJointOffset=<0.1 0 0>
     }]
   }
   n8 = Anim Dynamics - Bone: dyn_pjs_bot_r_A!(SimulationSpace:AnimPhysSimSpaceType Enum, ComponentPose=@n7, Alpha=1, LinearSpringConstant:float=30, ExternalForce="0,0,0") -> (Pose) {
    angularDampingOverride=0.9999
    angularSpringConstant=2000
    bAngularSpring=true
    bGravityOverrideInSimSpace=true
    bLinearSpring=true
    bOverrideAngularDamping=true
    bOverrideLinearDamping=true
    bUseGravityOverride=true
    linearDampingOverride=0.9999
    linearSpringConstant=30
    simulationSpace=RootRelative
    alphaBoolBlend {blendOption=Linear}
    boundBone {boneName=dyn_pjs_bot_r_A}
    componentAppliedLinearAccClamp=<1000 1000 1000>
    componentLinearAccScale=<1 1 1>
    componentLinearVelScale=<1 1 1>
    gravityOverride {z=-980}
    physicsBodyDefinitions [{
      boundBone {boneName=dyn_pjs_bot_r_A}
      boxExtents=<7 7 7>
      constraintSetup {angularLimitsMax=<5 5 5>, angularLimitsMin=<-5 -10 -5>, angularTarget=<0 0 0>, linearAxesMax=<2 2 4>, linearAxesMin=<-2 -2 0>}
      localJointOffset=<0.1 0 0>
     }]
   }
   n9 = Anim Dynamics - Bone: dyn_pjs_bot_r_B!(SimulationSpace:AnimPhysSimSpaceType Enum, ComponentPose=@n8, Alpha=1, LinearSpringConstant:float=30, ExternalForce="0,0,0") -> (Pose) {
    angularDampingOverride=0.9999
    angularSpringConstant=2000
    bAngularSpring=true
    bGravityOverrideInSimSpace=true
    bLinearSpring=true
    bOverrideAngularDamping=true
    bOverrideLinearDamping=true
    bUseGravityOverride=true
    linearDampingOverride=0.9999
    linearSpringConstant=30
    simulationSpace=RootRelative
    alphaBoolBlend {blendOption=Linear}
    boundBone {boneName=dyn_pjs_bot_r_B}
    componentAppliedLinearAccClamp=<1000 1000 1000>
    componentLinearAccScale=<1 1 1>
    componentLinearVelScale=<1 1 1>
    gravityOverride {z=-980}
    physicsBodyDefinitions [{
      boundBone {boneName=dyn_pjs_bot_r_B}
      boxExtents=<7 7 7>
      constraintSetup {angularLimitsMax=<5 5 5>, angularLimitsMin=<-5 -10 -5>, angularTarget=<0 0 0>, linearAxesMax=<2 2 4>, linearAxesMin=<-2 -2 0>}
      localJointOffset=<0.1 0 0>
     }]
   }
   n10 = Anim Dynamics - Bone: dyn_pjs_bot_r_C!(SimulationSpace:AnimPhysSimSpaceType Enum, ComponentPose=@n9, Alpha=1, LinearSpringConstant:float=30, ExternalForce="0,0,0") -> (Pose) {
    angularDampingOverride=0.9999
    angularSpringConstant=2000
    bAngularSpring=true
    bGravityOverrideInSimSpace=true
    bLinearSpring=true
    bOverrideAngularDamping=true
    bOverrideLinearDamping=true
    bUseGravityOverride=true
    linearDampingOverride=0.9999
    linearSpringConstant=30
    simulationSpace=RootRelative
    alphaBoolBlend {blendOption=Linear}
    boundBone {boneName=dyn_pjs_bot_r_C}
    componentAppliedLinearAccClamp=<1000 1000 1000>
    componentLinearAccScale=<1 1 1>
    componentLinearVelScale=<1 1 1>
    gravityOverride {z=-980}
    physicsBodyDefinitions [{
      boundBone {boneName=dyn_pjs_bot_r_C}
      boxExtents=<7 7 7>
      constraintSetup {angularLimitsMax=<5 5 5>, angularLimitsMin=<-5 -10 -5>, angularTarget=<0 0 0>, linearAxesMax=<2 2 4>, linearAxesMin=<-2 -2 0>}
      localJointOffset=<0.1 0 0>
     }]
   }
   n11 = Anim Dynamics - Bone: dyn_pjs_bot_r_D!(SimulationSpace:AnimPhysSimSpaceType Enum, ComponentPose=@n10, Alpha=1, LinearSpringConstant:float=30, ExternalForce="0,0,0") -> (Pose) {
    angularDampingOverride=0.9999
    angularSpringConstant=2000
    bAngularSpring=true
    bGravityOverrideInSimSpace=true
    bLinearSpring=true
    bOverrideAngularDamping=true
    bOverrideLinearDamping=true
    bUseGravityOverride=true
    linearDampingOverride=0.9999
    linearSpringConstant=30
    simulationSpace=RootRelative
    alphaBoolBlend {blendOption=Linear}
    boundBone {boneName=dyn_pjs_bot_r_D}
    componentAppliedLinearAccClamp=<1000 1000 1000>
    componentLinearAccScale=<1 1 1>
    componentLinearVelScale=<1 1 1>
    gravityOverride {z=-980}
    physicsBodyDefinitions [{
      boundBone {boneName=dyn_pjs_bot_r_D}
      boxExtents=<7 7 7>
      constraintSetup {angularLimitsMax=<5 5 5>, angularLimitsMin=<-5 -10 -5>, angularTarget=<0 0 0>, linearAxesMax=<2 2 4>, linearAxesMin=<-2 -2 0>}
      localJointOffset=<0.1 0 0>
     }]
   }
   n12 = Anim Dynamics - Bone: dyn_pjs_bot_r_E!(SimulationSpace:AnimPhysSimSpaceType Enum, ComponentPose=@n11, Alpha=1, LinearSpringConstant:float=30, ExternalForce="0,0,0") -> (Pose) {
    angularDampingOverride=0.9999
    angularSpringConstant=2000
    bAngularSpring=true
    bGravityOverrideInSimSpace=true
    bLinearSpring=true
    bOverrideAngularDamping=true
    bOverrideLinearDamping=true
    bUseGravityOverride=true
    linearDampingOverride=0.9999
    linearSpringConstant=30
    simulationSpace=RootRelative
    alphaBoolBlend {blendOption=Linear}
    boundBone {boneName=dyn_pjs_bot_r_E}
    componentAppliedLinearAccClamp=<1000 1000 1000>
    componentLinearAccScale=<1 1 1>
    componentLinearVelScale=<1 1 1>
    gravityOverride {z=-980}
    physicsBodyDefinitions [{
      boundBone {boneName=dyn_pjs_bot_r_E}
      boxExtents=<7 7 7>
      constraintSetup {angularLimitsMax=<5 5 5>, angularLimitsMin=<-5 -10 -5>, angularTarget=<0 0 0>, linearAxesMax=<2 2 4>, linearAxesMin=<-2 -2 0>}
      localJointOffset=<0.1 0 0>
     }]
   }
   n13 = Anim Dynamics - Bone: dyn_pjs_bot_r_F!(SimulationSpace:AnimPhysSimSpaceType Enum, ComponentPose=@n12, Alpha=1, LinearSpringConstant:float=30, ExternalForce="0,0,0") -> (Pose) {
    angularDampingOverride=0.9999
    angularSpringConstant=2000
    bAngularSpring=true
    bGravityOverrideInSimSpace=true
    bLinearSpring=true
    bOverrideAngularDamping=true
    bOverrideLinearDamping=true
    bUseGravityOverride=true
    linearDampingOverride=0.9999
    linearSpringConstant=30
    simulationSpace=RootRelative
    alphaBoolBlend {blendOption=Linear}
    boundBone {boneName=dyn_pjs_bot_r_F}
    componentAppliedLinearAccClamp=<1000 1000 1000>
    componentLinearAccScale=<1 1 1>
    componentLinearVelScale=<1 1 1>
    gravityOverride {z=-980}
    physicsBodyDefinitions [{
      boundBone {boneName=dyn_pjs_bot_r_F}
      boxExtents=<7 7 7>
      constraintSetup {angularLimitsMax=<5 5 5>, angularLimitsMin=<-5 -10 -5>, angularTarget=<0 0 0>, linearAxesMax=<2 2 4>, linearAxesMin=<-2 -2 0>}
      localJointOffset=<0.1 0 0>
     }]
   }
   Output Pose(Result=Layered blend per bone(
     BasePose=Component To Local(ComponentPose=@n13)
     BlendPoses_0=Slot 'DefaultSlot'()
     BlendWeights_0=Get CustomFELobbyIdleFloat() -> (CustomFELobbyIdleFloat:float)
    ) {layerSetup [{branchFilters [{boneName=calf_twist_02_r}]}]}) {layerGroup=None}
  ]}
 EventGraph {nodes [
   Event BlueprintUpdateAnimation()
   > n1 = Branch(Condition=Get bIsSkyDiving())
   > Set PantBottomDynSpringConstant(PantBottomDynSpringConstant:double=2000) -> (Output_Get:double)
   > Set PantMidDynSpringConstant(PantMidDynSpringConstant:double=1000) -> (Output_Get:double)
   Set PantBottomDynSpringConstant(execute=@n1.else, PantBottomDynSpringConstant=50)
   > Set PantMidDynSpringConstant(PantMidDynSpringConstant=100)
   Event OnCharacterCustomizationLoaded()
   > n2 = Branch(Condition=Get bIsInFrontEnd())
   > Set FrontendAlpha(FrontendAlpha:double=0.3) -> (Output_Get:double)
   > Set Simulation Space(Simulation Space:AnimPhysSimSpaceType Enum=World) -> (Output_Get:AnimPhysSimSpaceType Enum)
   > Set CustomFELobbyIdleFloat(CustomFELobbyIdleFloat:float=GetCurveValue(self=GetOwnerFrontEndAnimInstance(), CurveName=FrontEnd_Idle_Bypass)) -> (Output_Get:float)
   Set FrontendAlpha(execute=@n2.else, FrontendAlpha=1)
   > Set Simulation Space()
  ]}
]
variables [
 CustomFELobbyIdleFloat {type=number}
 FrontendAlpha {type=number}
 PantBottomDynSpringConstant {type=number}
 PantMidDynSpringConstant {type=number}
 Simulation Space {type=AnimPhysSimSpaceType}
]
