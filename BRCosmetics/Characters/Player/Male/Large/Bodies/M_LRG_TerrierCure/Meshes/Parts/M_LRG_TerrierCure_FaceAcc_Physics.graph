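# Physics asset for the TerrierCure large male character's face accessory, covering a bone-driven hat attached to the head.
# A head sphere and capsules on the hat bones form the bodies, with constraints chaining head, dyn_hat_mstr, then dyn_hat_01 through 05.
# The two root joints swing and twist freely, while the joints further along have swing limits and tighter twist limits, so the hat bends more than it twists.
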PhysicsAsset /BRCosmetics/Characters/Player/Male/Large/Bodies/M_LRG_TerrierCure/Meshes/Parts/M_LRG_TerrierCure_FaceAcc_Physics
bodies {
 dyn_hat_01 [{shapeType=Capsule, length=1.2577, radius=4.2932, shapeName=dyn_hat_01_capsule, center=<-0.3942 -0.0137 -3.4683>}]
 dyn_hat_02 [{shapeType=Capsule, length=1.526, radius=0.6466, center=<1.5 0 0.3>, rotation=<rot 90 0.0001 0>}]
 dyn_hat_03 [{shapeType=Capsule, length=1.0446, radius=0.505, shapeName=dyn_hat_03_capsule, center=<1.25 0 0.2>, rotation=<rot 90 0 0>}]
 dyn_hat_04 [{shapeType=Capsule, length=2.9131, radius=1.2023, center=<2.8346 -0.0111 0.0327>, rotation=<rot 0 -90.2523 -89.0326>}]
 dyn_hat_05 [{shapeType=Capsule, length=2.7938, radius=0.9621, center=<1.5612 0 0>, rotation=<rot 90 0 0>}]
 dyn_hat_mstr [{shapeType=Capsule, length=0.505, radius=0.505, shapeName=dyn_hat_mstr_capsule, rotation=<rot 0 -90 0>}]
 head [{shapeType=Sphere, radius=10.7683, shapeName=head_sphere, center=<7.2623 0 0>}]
}
constraints [
 {bone1Name=dyn_hat_02, bone2Name=dyn_hat_01, swing1LimitDegrees=30, swing2LimitDegrees=30, twistLimitDegrees=10}
 {bone1Name=dyn_hat_04, bone2Name=dyn_hat_03, swing1LimitDegrees=50, swing2LimitDegrees=50, twistLimitDegrees=5}
 {bone1Name=dyn_hat_05, bone2Name=dyn_hat_04, swing1LimitDegrees=50, swing2LimitDegrees=50, twistLimitDegrees=5}
 {
  bone1Name=dyn_hat_mstr
  bone2Name=head
  swing1LimitDegrees=45
  swing1Motion=Free
  swing2LimitDegrees=45
  swing2Motion=Free
  twistLimitDegrees=45
  twistMotion=Free
 }
 {
  bone1Name=dyn_hat_01
  bone2Name=dyn_hat_mstr
  swing1LimitDegrees=45
  swing1Motion=Free
  swing2LimitDegrees=45
  swing2Motion=Free
  twistLimitDegrees=45
  twistMotion=Free
 }
 {bone1Name=dyn_hat_03, bone2Name=dyn_hat_02, swing1LimitDegrees=50, swing2LimitDegrees=50, twistLimitDegrees=10}
]
